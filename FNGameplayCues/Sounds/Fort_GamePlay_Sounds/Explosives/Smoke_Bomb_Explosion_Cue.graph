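# Sound cue for a smoke bomb explosion, played after a short random delay.
# A distance crossfade blends a close layer that fades out early with a far layer that carries much farther.
# The close layer mixes a ninja smoke bomb with a roman candle fire sound at raised pitch.
# The far layer adds another delay, then picks one of two mono smoke bomb waves with boosted volume and slight pitch variation.
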SoundCue /FNGameplayCues/Sounds/Fort_GamePlay_Sounds/Explosives/Smoke_Bomb_Explosion_Cue
root=Delay_0
nodes [
 {id=Delay_0, children [DistanceCrossFade_1], properties {delayMax=0.2, delayMin=0.1}}
 {id=Delay_1, children [Random_3], properties {delayMax=0.1, delayMin=0.1}}
 {
  id=DistanceCrossFade_1
  children [Modulator_0,Modulator_2]
  properties {crossFadeInput [
    {fadeInDistanceEnd=0, fadeInDistanceStart=0, fadeOutDistanceEnd=800, fadeOutDistanceStart=400, volume=1}
    {fadeInDistanceEnd=400, fadeInDistanceStart=200, fadeOutDistanceEnd=7000, fadeOutDistanceStart=5000, volume=0.7}
   ]}
 }
 {id=Mixer_3, children [Modulator_4, WavePlayer_12], properties {inputVolume [0.6,1]}}
 {id=Modulator_0, children [Mixer_3], properties {pitchMax=1.4, pitchMin=1.2}}
 {id=Modulator_2, children [Delay_1], properties {pitchMax=1.1, pitchMin=0.9, volumeMax=2, volumeMin=2}}
 {id=Modulator_4, children [WavePlayer_10], properties {pitchMax=1, pitchMin=1}}
 {id=Random_3, children [WavePlayer_3, WavePlayer_11], properties {weights [1,1]}}
 {id=WavePlayer_10, properties {soundWaveAssetPtr=/Game/Sounds/Fort_Gadgets/Ninja/smokebomb_02}}
 {id=WavePlayer_11, properties {soundWaveAssetPtr=&/smokebomb_mono_02}}
 {id=WavePlayer_12, properties {soundWaveAssetPtr=/Game/Sounds/Fort_GamePlay_Sounds/Explosives/roman_candle_fire_02}}
 {id=WavePlayer_3, properties {soundWaveAssetPtr=&/smokebomb_mono_01}}
]
properties {
 attenuationSettings=/Game/Sounds/Attenuation/Fort_SkillUse_Att
 bOverrideConcurrency=true
 maxDistance=7000
 soundClassObject=&Fort_WeaponMelee_PSM
 volumeMultiplier=2
 concurrencyOverrides {resolutionRule=PreventNew}
}
